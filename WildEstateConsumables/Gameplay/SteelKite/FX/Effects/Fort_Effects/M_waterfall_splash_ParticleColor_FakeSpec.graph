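# Unlit translucent Niagara sprite and mesh particle material for waterfall splashes, colored by particle color through emissive adjustments.
# The splash texture's mip-biased green channel is shaped by the dynamic Erode parameter, then rounded into a hard-edged mask.
# Particle alpha scales that mask, with depth fade (skipped on mobile) and near-camera fade.
# Optional world position offset follows water displacement plus a Z offset.
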Material /WildEstateConsumables/Gameplay/SteelKite/FX/Effects/Fort_Effects/M_waterfall_splash_ParticleColor_FakeSpec
graph {
 properties {
  bTangentSpaceNormal=false
  bUsedWithNiagaraMeshParticles=true
  bUsedWithNiagaraSprites=true
  blendMode=BLEND_Translucent
  pixelDepthOffsetMode=PDOM_Legacy
  shadingModel=MSM_Unlit
  translucencyLightingMode=TLM_VolumetricPerVertexNonDirectional
  translucencyPass=MTP_BeforeDOF
  translucentSelfShadowDensityScale=0
  translucentSelfShadowSecondDensityScale=0
  translucentShadowDensityScale=0
  usedShadingModels=Unlit
 }
 nodes [
  n1 = ParticleColor()
  MaterialFunctionCall(Emissive=@n1.RGB) {materialFunction=&MF_FX_EmissiveAdjustments}
  n2 = TextureSample!(UVs, Tex, Bias=ScalarParameter() {defaultValue=2, parameterName=Bias}, Apply View MipBias) {mipValueMode=TMVM_MipBias, samplerType=SAMPLERTYPE_LinearColor, texture=/Game/Effects/Fort_Effects/Textures/Water/T_waterSplash}
  n3 = DynamicParameter() {paramNames [Erode,Param2,Param3,Param4]}
  n4 = Power(Base=@n2.G, Exp=LinearInterpolate(Alpha=@n3.Erode) {constA=5, constAlpha=0}) {constExponent=50}
  n5 = LinearInterpolate(Alpha=@n3.Erode) {constAlpha=0, constB=200}
  n6 = Multiply(A=@n2.G, B=Clamp(None=Multiply(A=@n4, B=@n5)))
  n7 = Multiply(A=@n1.A, B=Round(None=Add(A=@n6) {constB=0.45}))
  n8 = ScalarParameter() {defaultValue=100, parameterName=Depth Fade}
  Multiply(
   A=ShadingPathSwitch(Default=DepthFade(Opacity=@n7, FadeDistance=@n8), Mobile=@n7) {desc=Migrated from FeatureLevelSwitch}
   B=MaterialFunctionCall(Distance=ScalarParameter() {parameterName=CameraFade}) {materialFunction=&Near_Camera_Fade}.Result
  )
  n9 = Multiply(
   A=MaterialFunctionCall!(Ignore XY Offsets=StaticBoolParameter() {defaultValue=true, parameterName=Ignore XY Gerstner Offsets}, Position, WaterBodyIndex, Depth) {materialFunction=/FortWater/Materials/Functions/WaterParticleDisplacement}.WPO
   B=VectorParameter() {parameterName=Water Displacement Multiplier, defaultValue=<rgba 1 1 1 0>}.RGB
  )
  n10 = MaterialFunctionCall(Z=ScalarParameter() {defaultValue=1, parameterName=ZOffset}) {materialFunction=&MakeFloat3}
  StaticSwitchParameter(
   True=LinearInterpolate(A=Constant(), B=Add(A=@n9, B=@n10.Result), Alpha=ScalarParameter() {defaultValue=1, parameterName=UseWaterDisplacement})
   False=Constant()
  ) {defaultValue=true, parameterName=UseWPO}
 ]
}
